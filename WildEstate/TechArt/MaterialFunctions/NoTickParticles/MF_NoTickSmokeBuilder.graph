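# Material function that builds a procedural smoke puff for NoTick particles, outputting color, alpha and a generated world-space normal.
# Two noise samples, panned differently and randomized per particle in time, tiling and offset, warp the UVs of a radial mask, clipped inside the UV edge for alpha.
# Color blends start to end tint over lifetime with desaturated shading.
# The warped UVs also form a camera-facing hemisphere normal, with a boost for wrapped lighting.
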MaterialFunction /WildEstate/TechArt/MaterialFunctions/NoTickParticles/MF_NoTickSmokeBuilder
graph {nodes [
  n1 = VectorParameter() {group=Smoke Builder, parameterName=SmokeTintStart, sortPriority=12, defaultValue=<rgba 0.6562 0.6262 0.5708 1>}
  n2 = VectorParameter() {group=Smoke Builder, parameterName=SmokeTintEnd, sortPriority=13, defaultValue=<rgba 0.875 0.835 0.7611 1>}
  n3 = FunctionInput() {bUsePreviewValueAsDefault=true, inputName=Lifetime, inputType=FunctionInput_Scalar, sortPriority=1}
  n4 = Desaturation(None=NamedRerouteUsage())
  FunctionOutput(None=Multiply(A=LinearInterpolate(A=@n1.RGB, B=@n2.RGB, Alpha=@n3), B=LinearInterpolate(Alpha=@n4) {constA=0.5})) {outputName=Color}
  n5 = NamedRerouteUsage()
  n6 = Time()
  n7 = FunctionInput() {bUsePreviewValueAsDefault=true, inputName=PerParticleRandom, sortPriority=2, previewValue=<0.1 0.1 0 1>}
  n8 = Multiply(
   A=Add(
    A=MaterialFunctionCall(f3=@n7) {materialFunction=&AddComponents}.f3
    B=FunctionInput(Preview=@n6) {bUsePreviewValueAsDefault=true, inputName=Time, inputType=FunctionInput_Scalar, sortPriority=5}
   ) {desc=Offset time per particle}
   B=ScalarParameter() {defaultValue=0.75, group=Smoke Builder, parameterName=Smoke Pan Speed}
  )
  n9 = NamedRerouteUsage()
  n10 = MaterialFunctionCall(Float3=@n7) {materialFunction=&BreakOutFloat3Components}
  n11 = LinearInterpolate(Alpha=@n10.B) {constA=0.8, constB=1.2}
  n12 = Add(
   A=Multiply(A=Multiply(A=@n9, B=@n11), B=ScalarParameter() {defaultValue=1, group=Smoke Builder, parameterName=Smoke Tiling, sortPriority=5})
   B=AppendVector(A=@n10.R, B=@n10.R)
  )
  n13 = FunctionInput() {bUsePreviewValueAsDefault=true, inputName=Direction, inputType=FunctionInput_Vector2, sortPriority=4, previewValue=<0.1 0.1 0 1>}
  n14 = Panner(Coordinate=@n12, Time=@n8, Speed=@n13)
  n15 = TextureObjectParameter() {group=Smoke Builder, parameterName=SmokeTextureMask, sortPriority=4, texture=&AtmosphericCloudNoise01}
  n16 = MaterialFunctionCall(Float2=@n13) {materialFunction=&BreakOutFloat2Components}
  n17 = Panner(
   Coordinate=Multiply(A=@n12) {constB=0.66}
   Time=@n8
   Speed=AppendVector(A=Multiply(A=@n16.R) {constB=-0.6}, B=Multiply(A=@n16.G) {constB=0.8})
  )
  n18 = NamedRerouteDeclaration(None=AppendVector(A=TextureSample(UVs=@n14, Tex=@n15).R, B=TextureSample(UVs=@n17, Tex=@n15).R)) {name=NoiseUVs, nodeColor=<rgba 1 0.4453 0 1>}
  n19 = ScalarParameter() {defaultValue=0.2, group=Smoke Builder, parameterName=Smoke UV Warp, sortPriority=6}
  n20 = Multiply(A=Subtract(A=Multiply(A=@n18) {constB=2}), B=@n19) {constB=0.2}
  n21 = Divide(A=@n19)
  n22 = Constant2Vector() {g=0.5, r=0.5}
  n23 = SphereMask(
   A=Add(A=Add(A=@n5, B=@n20), B=@n21)
   B=@n22
   Radius=ScalarParameter() {defaultValue=0.45, group=Smoke Builder, parameterName=Smoke Radius, sortPriority=8}
  ) {attenuationRadius=0.45, hardnessPercent=1}
  n24 = SphereMask(A=@n5, B=@n22) {attenuationRadius=0.49, desc=Ensure we dont scale radual passed the edge of the UVs, hardnessPercent=1}
  FunctionOutput(None=Saturate(None=Multiply(A=@n23, B=@n24))) {outputName=Alpha, sortPriority=1}
  n25 = Add(A=@n20, B=@n21)
  n26 = FunctionInput(Preview=TextureCoordinate()) {
   bUsePreviewValueAsDefault=true
   desc=Only needed for using the generatewd normal if UVs are rotated or adjusted in any way. Defaults Tex Coord 0
   inputName=UVs
   inputType=FunctionInput_Vector2
   sortPriority=6
  }
  n27 = NamedRerouteDeclaration(None=@n26) {name=UVs, nodeColor=<rgba 0 1 0.7891 1>}
  n28 = Multiply(A=Add(A=Multiply(A=@n25) {constB=0.3}, B=Subtract(A=@n27) {constB=0.5})) {constB=2}
  n29 = SquareRoot(None=OneMinus(None=DotProduct(A=@n28, B=@n28)))
  n30 = Multiply(A=@n29) {constB=2}
  n31 = ScalarParameter() {
   desc="Allows for more light to \"wrap\" around the normals, letting more light to hit the backside more"
   group=Smoke Builder
   parameterName=NormalShapeBoost
   sortPriority=40
  }
  n32 = AppendVector(A=@n28, B=LinearInterpolate(A=@n29, B=Subtract(A=@n30), Alpha=@n31))
  n33 = Constant3Vector() {constant=<rgba 1 -1 -1 0>}
  FunctionOutput(None=Transform(None=Multiply(A=@n32, B=@n33)) {transformSourceType=TRANSFORMSOURCE_Camera}) {outputName=GeneratedNormalWS, sortPriority=1}
 ]}
