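# Material function for no-tick particles that outputs a ShrinkMultiplier for shrinking them as they get close to the camera.
# It divides pixel depth by the Shrink.NearStartDistance parameter and saturates the result, ramping linearly from 0 at the camera to 1 at that distance and beyond.
# The Shrink.UseDistanceShrink static switch selects this ramp or a constant 1, so disabled shrinking leaves the multiplier neutral.
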MaterialFunction /WildEstate/TechArt/MaterialFunctions/NoTickParticles/MF_NoTickParticle_DistanceShrink
graph {nodes [
  n1 = Divide(A=PixelDepth().R, B=ScalarParameter() {defaultValue=3500, group=Shrink, parameterName=Shrink.NearStartDistance})
  FunctionOutput(None=StaticSwitch(
    True=Saturate(None=@n1)
    False=Constant() {r=1}
    Value=StaticBoolParameter() {group=Shrink, parameterName=Shrink.UseDistanceShrink, sortPriority=31}
   )) {outputName=ShrinkMultiplier}
 ]}
