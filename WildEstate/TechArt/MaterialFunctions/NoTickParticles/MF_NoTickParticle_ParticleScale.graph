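# Material function for NoTick particles that outputs a 2D scale over each particle's lifetime.
# Lifetime, shaped by a power curve, blends randomized sizes, with a static switch choosing two-point or start/mid/end interpolation, then a master multiplier applies.
# Size variance comes from fractional sums of PerParticleRandom channel pairs.
# Optional fake spin oscillates X scale with a time-driven sine, with per-particle phase, speed variance and chance.
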MaterialFunction /WildEstate/TechArt/MaterialFunctions/NoTickParticles/MF_NoTickParticle_ParticleScale
graph {nodes [
  n1 = ScalarParameter() {defaultValue=0.05, desc=Thinnest the particle will get, group=Scale, parameterName=SpinMinScale, sliderMax=1, sortPriority=57}
  n2 = Time()
  n3 = ScalarParameter() {defaultValue=0.5, desc="Spin Speed, With 10% variance =/-", group=Scale, parameterName=SpinSpeed, sliderMax=1, sortPriority=53}
  n4 = ScalarParameter() {defaultValue=0.1, desc="Spin Speed Variance =/-", group=Scale, parameterName=SpinSpeedVariance, sliderMax=1, sortPriority=54}
  n5 = Multiply(A=@n3, B=@n4) {constB=0.1}
  n6 = Add(A=@n3, B=LinearInterpolate(A=Multiply(A=@n5) {constB=-1}, B=@n5, Alpha=NamedRerouteUsage()))
  n7 = MaterialFunctionCall(f3=NamedRerouteUsage()) {materialFunction=&AddComponents}
  n8 = Sine(None=Add(A=Multiply(A=@n2, B=@n6), B=@n7.f3))
  n9 = AppendVector(A=LinearInterpolate(A=@n1, Alpha=ConstantBiasScale(None=@n8)) {constA=0.05}, B=Constant() {r=1})
  n10 = ScalarParameter() {
   defaultValue=1
   desc="1 = linear scale froms tart to finish, higher values will create an exponential curve, staying the start size longer.\r\nResults my differ when using Lerp 3 Scaling"
   group=Scale
   parameterName=ParticleSizeCurve
   sliderMax=10
   sliderMin=0.5
   sortPriority=30
  }
  n11 = Power(Base=FunctionInput() {inputName=Lifetime, inputType=FunctionInput_Scalar}, Exp=@n10)
  n12 = Multiply(A=@n11) {constB=2}
  n13 = LinearInterpolate(A=NamedRerouteUsage(), B=NamedRerouteUsage(), Alpha=Saturate(None=@n12))
  n14 = NamedRerouteUsage()
  n15 = Saturate(None=Subtract(A=@n12))
  n16 = NamedRerouteUsage()
  n17 = NamedRerouteUsage()
  n18 = ScalarParameter() {defaultValue=1, group=Scale, parameterName=ParticleScaleMultiMaster, sliderMax=10, sliderMin=0.5, sortPriority=30}
  n19 = Multiply(
   A=StaticSwitchParameter(True=LinearInterpolate(A=@n13, B=@n14, Alpha=@n15), False=LinearInterpolate(A=@n16, B=@n17, Alpha=@n11)) {group=Scale, parameterName=UseLerp3Scaling, sortPriority=15}
   B=@n18
  )
  n20 = ScalarParameter() {
   desc=Scales the particle on the X axis in and out to simulate the particle spinning
   group=Scale
   parameterName=UseSpinning
   sliderMax=1
   sortPriority=50
  }
  n21 = Frac(None=@n7.f3)
  n22 = ScalarParameter() {
   desc="Chance particles have to spin.\r\n0 = no spin\r\n1 = every particle spins"
   group=Scale
   parameterName=PerParticleSpinChance
   sliderMax=1
   sortPriority=51
  }
  n23 = LinearInterpolate(Alpha=@n22) {constA=-1}
  n24 = Round(None=Saturate(None=Add(A=@n21, B=@n23)))
  FunctionOutput(None=LinearInterpolate(A=@n19, B=Multiply(A=@n19, B=@n9), Alpha=Multiply(A=@n20, B=@n24))) {outputName="Scale (V2)"}
  n25 = ScalarParameter() {defaultValue=1, group=Scale, parameterName=ParticleMidSize, sortPriority=29}
  n26 = ScalarParameter() {desc=Increases/Decreases size by 0-100 Percent per particle, group=Scale, parameterName=ParticleMidVariance, sliderMax=1, sortPriority=42}
  n27 = Multiply(A=@n25, B=@n26)
  n28 = Multiply(A=@n27) {constB=-1}
  n29 = NamedRerouteUsage()
  NamedRerouteDeclaration(None=Add(A=@n25, B=LinearInterpolate(A=@n28, B=@n27, Alpha=@n29))) {name=Mid, nodeColor=<rgba 0 0.4375 1 1>}
  n30 = ScalarParameter() {defaultValue=1, group=Scale, parameterName=ParticleStartSize, sortPriority=28}
  n31 = ScalarParameter() {desc=Increases/Decreases size by 0-100 Percent per particle, group=Scale, parameterName=ParticleStartVariance, sliderMax=1, sortPriority=41}
  n32 = Multiply(A=@n30, B=@n31)
  n33 = Multiply(A=@n32) {constB=-1}
  n34 = NamedRerouteUsage()
  NamedRerouteDeclaration(None=Add(A=@n30, B=LinearInterpolate(A=@n33, B=@n32, Alpha=@n34))) {name=Start, nodeColor=<rgba 0.7578 1 0 1>}
  n35 = ScalarParameter() {defaultValue=1, group=Scale, parameterName=ParticleEndSize, sortPriority=30}
  n36 = ScalarParameter() {desc=Increases/Decreases size by 0-100 Percent per particle, group=Scale, parameterName=ParticleEndVariance, sliderMax=1, sortPriority=43}
  n37 = Multiply(A=@n35, B=@n36)
  n38 = Multiply(A=@n37) {constB=-1}
  n39 = NamedRerouteUsage()
  NamedRerouteDeclaration(None=Add(A=@n35, B=LinearInterpolate(A=@n38, B=@n37, Alpha=@n39))) {name=End, nodeColor=<rgba 0.9219 0 1 1>}
  n40 = NamedRerouteDeclaration(None=FunctionInput() {bUsePreviewValueAsDefault=true, desc=Used for Scale Variance, inputName=PerParticleRandom, sortPriority=1}) {name=PerParticleRand, nodeColor=<rgba 0 0.2031 1 1>}
  n41 = ComponentMask(None=@n40) {g=true, r=true}
  NamedRerouteDeclaration(None=Frac(None=MaterialFunctionCall(f2=@n41) {materialFunction=&AddComponents}.f2)) {name=StartRand, nodeColor=<rgba 0 1 0.6016 1>}
  n42 = ComponentMask(None=@n40) {b=true, g=true}
  NamedRerouteDeclaration(None=Frac(None=MaterialFunctionCall(f2=@n42) {materialFunction=&AddComponents}.f2)) {name=MidRand, nodeColor=<rgba 1 0.2812 0 1>}
  n43 = ComponentMask(None=@n40) {b=true, r=true}
  NamedRerouteDeclaration(None=Frac(None=MaterialFunctionCall(f2=@n43) {materialFunction=&AddComponents}.f2)) {name=EndRand, nodeColor=<rgba 0 0.0391 1 1>}
 ]}
