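# Material function for NoTickParticles that pushes particle spawn positions away from the center toward the bounds.
# Push distance scales with bounds length and inverse normalized distance from center, so central spawns move most, varied by averaged random data between min and max multiplier parameters.
# SignedData supplies per-axis direction rather than the position vector, a rounded AxisMask parameter selects axes, and AlphaBlend fades between original and pushed positions.
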MaterialFunction /WildEstate/TechArt/MaterialFunctions/NoTickParticles/MF_PushParticleSpawns
graph {nodes [
  n1 = FunctionInput() {inputName=SignedData, sortPriority=4}
  n2 = Length(None=ComponentMask(None=FunctionInput() {inputName=Bounds, sortPriority=2}) {b=true, g=true, r=true}) {is=class MaterialExpressionLength}
  n3 = ComponentMask(None=FunctionInput() {inputName=PreParticleSpawnPos, sortPriority=1}) {b=true, g=true, r=true}
  n4 = Saturate(None=InverseLinearInterpolate(B=@n2, Value=Length(None=@n3)) {constB=0})
  n5 = ScalarParameter() {group=Particle Misc, parameterName=SpawnPush.MinRangeMulti, sortPriority=50}
  n6 = ScalarParameter() {defaultValue=1, group=Particle Misc, parameterName=SpawnPush.MaxRangeMulti, sortPriority=50}
  n7 = FunctionInput() {inputName=RandomData, sortPriority=3}
  n8 = Saturate(None=Divide(A=MaterialFunctionCall(f3=@n7) {materialFunction=&AddComponents}.f3) {constB=3})
  n9 = Multiply(A=@n2, B=Multiply(A=OneMinus(None=@n4), B=LinearInterpolate(A=@n5, B=@n6, Alpha=@n8) {constA=0.2, constB=0.7}))
  n10 = VectorParameter() {group=Particle Misc, parameterName=SpawnPush.AxisMask, sortPriority=50, defaultValue=<rgba 1 1 1 1>}
  n11 = Multiply(A=Multiply(A=Sign(None=@n1), B=Abs(None=@n9)), B=Round(None=Saturate(None=@n10.RGB)))
  n12 = ScalarParameter() {group=Particle Misc, parameterName=SpawnPush.AlphaBlend, sortPriority=50}
  FunctionOutput(None=LinearInterpolate(A=@n3, B=Add(A=@n3, B=@n11), Alpha=Saturate(None=@n12))) {description=Pushes the spawn of particles away from the center out to the max bounds, outputName=Partcile Spawn Pos}
 ]}
