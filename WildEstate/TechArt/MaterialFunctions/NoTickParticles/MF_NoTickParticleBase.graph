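# Base material function for "NoTick" particles, animating mesh quads statelessly through World Position Offset and outputting material attributes, color, opacity and UVs.
# Vertex color and seeded time give each particle a looping lifetime and per-loop random values from cell noise or sines.
# These drive bounded spawning, cone velocity with wind and gravity, optional billboarding, scale and rotation.
# Color and opacity follow gradients or curve atlas, with hue shift, depth and near-camera fades.
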MaterialFunction /WildEstate/TechArt/MaterialFunctions/NoTickParticles/MF_NoTickParticleBase
graph {
 properties {previewBlendMode=BLEND_Translucent}
 nodes [
  FunctionOutput(None=NamedRerouteUsage()) {description="A Vector 3 value that differs per particle.\r\n\r\nAlso good for debug colors", outputName=PerParticleRandom, sortPriority=13}
  FunctionOutput(None=NamedRerouteUsage()) {outputName=WPO, sortPriority=4}
  FunctionOutput(None=NamedRerouteUsage()) {description="Outputs alpha value of the Particle Age (0-1)", outputName=Lifetime, sortPriority=10}
  n1 = NamedRerouteUsage()
  FunctionOutput(None=MakeMaterialAttributes!(
    BaseColor=@n1
    Metallic
    Specular
    Roughness
    Anisotropy
    EmissiveColor=@n1
    Opacity
    OpacityMask=NamedRerouteUsage()
    Normal
    Tangent
    WorldPositionOffset=NamedRerouteUsage()
    SubsurfaceColor
    ClearCoat
    ClearCoatRoughness
    AmbientOcclusion
    Refraction
    CustomizedUVs_0
    CustomizedUVs_1
    CustomizedUVs_2
    CustomizedUVs_3
    CustomizedUVs_4
    CustomizedUVs_5
    CustomizedUVs_6
    CustomizedUVs_7
    PixelDepthOffset
    ShadingModel
    Displacement
   )) {description="A base material that include WPO, Opacity, and Color", outputName=MaterialBase}
  FunctionOutput(None=NamedRerouteUsage()) {
   description="Outputs Opacity driven by Particle Lifetime. \r\n\r\nThis is only the BASE Opacity for fading in/out the particle. This should (likely) be mutliplied by the unique Particles Opacity."
   outputName=Opacity
   sortPriority=2
  }
  n2 = ScalarParameter() {group=Movement, parameterName=ParticleSpeedVariationAddititvePerActor}
  n3 = LinearInterpolate(A=Multiply(A=@n2) {constB=-1}, B=@n2, Alpha=Frac(None=NamedRerouteUsage()))
  n4 = ScalarParameter() {defaultValue=0.5, group=Movement, parameterName=ParticleSpeed}
  n5 = ObjectPositionWS()
  n6 = StaticSwitchParameter(
   True=Modulo!(A=Length(None=@n5) {is=class MaterialExpressionLength}, B=Constant() {r=9999})
   False=ScalarParameter() {desc="Seed (essentailly a time offset)", group=Particle Global, parameterName=Seed, sortPriority=33}
  ) {desc=Asset should be static / never move, group=Particle Global, parameterName=UsePerActorRandom, sortPriority=33}
  n7 = Add(A=Time(), B=@n6)
  n8 = LinearInterpolate(
   A=Multiply(A=Add(A=@n3, B=@n4), B=NamedRerouteDeclaration(None=@n7) {name=SeededTime, nodeColor=<rgba 0.6875 1 0 1>})
   B=Add(A=ScalarParameter() {group=Particle Global, parameterName=ManualTime, sortPriority=40}, B=@n6)
   Alpha=ScalarParameter() {group=Particle Global, parameterName=UseDebugTime, sortPriority=40}
  )
  FunctionOutput(None=@n8) {outputName=Time, sortPriority=11}
  FunctionOutput(None=NamedRerouteUsage()) {description=Outputs Color value at the current age of the particle, outputName=Color, sortPriority=3}
  n9 = FunctionInput(Preview=TextureCoordinate()) {bUsePreviewValueAsDefault=true, desc=UVs should be in 0-1 Space, inputName=UVs, inputType=FunctionInput_Vector2}
  n10 = MaterialFunctionCall(Color=NamedRerouteUsage()) {materialFunction=&SplitComponents}
  n11 = ScalarParameter() {
   defaultValue=1
   desc="Favor rotation one direction over the other. Or set to \"0\" for completely random"
   group=Rotation
   parameterName=RotationDirectionBias
   sliderMax=1
   sliderMin=-1
  }
  n12 = Sign(None=Add(A=@n10.R, B=@n11))
  n13 = ScalarParameter() {group=Rotation, parameterName=RotationOffsetVariation, sliderMax=1}
  n14 = NamedRerouteUsage()
  n15 = ScalarParameter() {desc=Degrees Per Second, group=Rotation, parameterName=RotationSpeed}
  n16 = Add(
   A=Multiply(A=LinearInterpolate(B=@n10.G, Alpha=@n13)) {constB=180}
   B=Add(A=Multiply(A=@n14, B=@n15), B=ScalarParameter() {group=Rotation, parameterName=RotationAngle})
  )
  n17 = MaterialFunctionCall(UVs=@n9, "Rotation Angle (0-1)"=Multiply(A=Multiply(A=@n12, B=@n16), B=Divide() {constB=360})) {materialFunction=&CustomRotator}
  FunctionOutput(None=StaticSwitch(True=@n17.Rotated Values, False=@n9, Value=NamedRerouteUsage())) {outputName=UVs, sortPriority=50}
  FunctionOutput(None=NamedRerouteUsage()) {outputName=ParticleSpawnPosition, sortPriority=14}
  FunctionOutput(None=NamedRerouteUsage()) {outputName=ActorRandomSeed, sortPriority=20}
  n18 = VertexColor()
  n19 = Multiply(A=Frac(None=@n18)) {constB=255}
  n20 = Add(A=@n8, B=@n18)
  n21 = Add(A=Floor(None=Add(A=@n19) {constB=0.5}), B=Floor(None=@n20))
  n22 = AppendVector(A=ComponentMask(None=Sine(None=@n21) {period=4.12}) {r=true}, B=ComponentMask(None=Sine(None=@n21) {period=3.3956}) {g=true})
  n23 = ComponentMask(None=Sine(None=@n21) {period=7.34}) {b=true}
  n24 = StaticSwitchParameter(True=ConstantBiasScale(None=AppendVector(A=@n22, B=@n23)), False=VectorNoise(World Position=@n21)) {
   desc="By default (FALSE) - Noise is generated by VectorNoise(CellNoise) \r\n\r\nUsing the alternate method (TRUE) uses SINE functions to generate Random Numbers. This is only slightly cheaper by less than 15-20 instructions\r\n\r\nNOTE: This was a fallback incase Vector Noise Shouldnt be used - but Cellnoise is Cheap so shouldnt matter"
   group=Particle Global
   parameterName=UseAlternateNoiseGeneration
   sortPriority=20
  }
  NamedRerouteDeclaration(None=@n24) {name=RandColor, nodeColor=<rgba 0.6406 0 1 1>}
  NamedRerouteDeclaration(None=Frac(None=@n20) {desc=Looping Speed}) {name=Lifetime, nodeColor=<rgba 0.6406 1 0 1>}
  n25 = CurveAtlasRowParameter(CurveTime=NamedRerouteUsage()) {desc=Curve to drive Color and Opacity over the Lifetime, group=Color, parameterName=CustomColorOpacityCurve}
  NamedRerouteDeclaration(None=@n25.A) {name=CurveOpacity, nodeColor=<rgba 0 1 0.2266 1>}
  n26 = FunctionInput(Preview=StaticBool() {value=true}) {
   bUsePreviewValueAsDefault=true
   desc="On by Default - Disabling this will keep all other WPO math without the camera facing billboards.\r\n\r\nmeaning Billboard logic (if any) should and will need to be handled outside this funciton"
   inputName=UseBillboardWPO
   inputType=FunctionInput_StaticBool
   sortPriority=6
  }
  NamedRerouteDeclaration(None=@n26) {name=UseBillboardBool, nodeColor=<rgba 0 0.1797 1 1>}
  NamedRerouteDeclaration(None=@n6) {name=ActorSeed, nodeColor=<rgba 0.6641 0 1 1>}
  n27 = VectorParameter() {group=Particle Global, parameterName=Bounds, channelNames {a=Multiplier, b="", g="", r=""}, defaultValue=<rgba 300 300 300 1>}
  n28 = Multiply(A=ComponentMask(None=@n27) {b=true, g=true, r=true}, B=@n27.Multiplier)
  NamedRerouteDeclaration(None=@n28) {name=SpawnBounds, nodeColor=<rgba 0.4062 0 1 1>}
  n29 = VectorParameter() {group=Forces, parameterName=Wind, sortPriority=13}
  n30 = FunctionInput() {
   bUsePreviewValueAsDefault=true
   desc="Multiply X by Wind (V3) - can create a sort of sway effect for example"
   inputName=WindModifierMult
   sortPriority=10
   previewValue=<1 0 1 1>
  }
  n31 = Add(A=Add(A=Multiply(A=@n29.RGB, B=@n30), B=VectorParameter() {group=Forces, parameterName=Gravity, sortPriority=12}.RGB), B=ObjectPositionWS())
  NamedRerouteDeclaration(None=Multiply(A=TransformPosition(None=@n31) {transformSourceType=TRANSFORMPOSSOURCE_World}, B=NamedRerouteUsage())) {name=ForcesOut, nodeColor=<rgba 0 0.0859 1 1>}
  n32 = ScalarParameter() {
   defaultValue=0.5
   desc="Curve based off the START of the particles LIFETIME.\r\n\r\n(Fades the particle in when spawned)"
   group=Opacity
   parameterName=OpacityFadeInPower
   sortPriority=13
  }
  n33 = ScalarParameter() {defaultValue=1, desc=Adjusts the curve for fading, group=Opacity, parameterName=OpacityFadeInMulti, sliderMin=1, sortPriority=14}
  n34 = Multiply(A=Power(Base=NamedRerouteUsage(), Exp=@n32), B=@n33)
  n35 = NamedRerouteUsage()
  n36 = ScalarParameter() {
   defaultValue=2
   desc="Curve based off the END of the particles LIFETIME.\r\n\r\n(Fades the particle out when near end of lifetime)"
   group=Opacity
   parameterName=OpacityFadeOutPower
   sortPriority=15
  }
  n37 = ScalarParameter() {defaultValue=1, desc=Adjusts the curve for fading, group=Opacity, parameterName=OpacityFadeOutMulti, sliderMin=1, sortPriority=16}
  n38 = Multiply(A=Power(Base=OneMinus(None=@n35), Exp=@n36), B=@n37)
  n39 = StaticSwitch(True=NamedRerouteUsage(), False=Multiply(A=Saturate(None=@n34), B=Saturate(None=@n38)), Value=NamedRerouteUsage())
  n40 = ScalarParameter() {defaultValue=4, group=Opacity, parameterName=Opacity, sortPriority=12}
  n41 = VertexInterpolator(VS=Saturate(None=Multiply(A=@n39, B=@n40)))
  n42 = ScalarParameter() {defaultValue=25, group=Opacity, parameterName=DepthFadeDistance, sortPriority=19}
  n43 = QualitySwitch(Default=FeatureLevelSwitch(Default=DepthFade(Opacity=@n41.PS, FadeDistance=@n42), ES2=@n41.PS), Low=@n41.PS, Medium=@n41.PS)
  n44 = FunctionInput(Preview=StaticBool()) {bUsePreviewValueAsDefault=true, inputName=UseDepthFadeInOpacity, inputType=FunctionInput_StaticBool, sortPriority=10}
  n45 = WorldPosition() {worldPositionShaderOffset=WPT_CameraRelative}
  n46 = Divide(
   A=Add(A=Length(None=@n45.XYZ), B=ScalarParameter() {group=Opacity, parameterName=NearFadeDistanceDistanceOffset, sortPriority=21})
   B=ScalarParameter() {defaultValue=50, group=Opacity, parameterName=NearFadeDistance, sortPriority=20}
  )
  NamedRerouteDeclaration(None=Multiply(A=StaticSwitch(True=@n43, False=@n41.PS, Value=@n44), B=Saturate(None=@n46))) {name=Opacity_OUT, nodeColor=<rgba 0 1 1 1>}
  n47 = Subtract(A=Multiply(A=NamedRerouteUsage()) {constB=2})
  NamedRerouteDeclaration(None=@n47) {name=RandomSignedData, nodeColor=<rgba 0.8984 1 0 1>}
  n48 = Multiply(
   A=MaterialFunctionCall!(Lifetime=NamedRerouteUsage(), PerParticleRandom=NamedRerouteUsage()) {materialFunction=&/MF_NoTickParticle_ParticleScale}."Scale (V2)"
   B=MaterialFunctionCall() {materialFunction=&/MF_NoTickParticle_DistanceShrink}.ShrinkMultiplier
  )
  n49 = ObjectPositionWS()
  n50 = WorldPosition()
  n51 = ComponentMask(None=@n48) {r=true}
  n52 = StaticSwitch(
   True=MaterialFunctionCall!(UVs=@n17.Rotated Values, Scale=@n48, UV Center) {materialFunction=/WildEstate/TechArt/MaterialFunctions/MF_Billboard}.Position
   False=LinearInterpolate(A=Subtract(A=@n49, B=@n50.XYZ), Alpha=Max(A=@n51) {constB=0}) {constAlpha=1, constB=0}
   Value=NamedRerouteUsage()
  )
  n53 = FunctionInput() {bUsePreviewValueAsDefault=true, inputName=BoundModifier, sortPriority=8, previewValue=<1 1 1 1>}
  n54 = MaterialFunctionCall!(
   PreParticleSpawnPos=Multiply(A=Multiply(A=@n47, B=@n28), B=@n53)
   Bounds=NamedRerouteUsage()
   RandomData=NamedRerouteUsage()
   SignedData=NamedRerouteUsage()
  ) {materialFunction=&/MF_PushParticleSpawns}
  n55 = ScalarParameter() {desc=Cone Width, group=Movement, parameterName=ParticleSpread, sliderMax=360}
  n56 = Constant() {r=360}
  n57 = Constant() {r=2}
  n58 = ComponentMask(None=VectorParameter() {group=Movement, parameterName=ParticleDirection, defaultValue=<rgba 0 1 0 1>}.RGB) {b=true, g=true, r=true}
  n59 = Add(
   A=Multiply(
    A=@n47
    B=MaterialFunctionCall(Input=@n55, Input High=@n56, Target High=@n57) {materialFunction=&RemapValueRange}.Result
   )
   B=StaticSwitchParameter(True=Transform(None=@n58) {transformSourceType=TRANSFORMSOURCE_World, transformType=TRANSFORM_Local}, False=@n58) {group=Movement, parameterName=IsParticleDirectionWorldSpace}
  )
  n60 = ScalarParameter() {desc=Velocity at beginning of lifetime, group=Movement, parameterName=ParticleStartVelocity, sortPriority=12}
  n61 = ScalarParameter() {desc=Velocity at end of lifetime, group=Movement, parameterName=ParticleEndVelocity, sortPriority=13}
  n62 = NamedRerouteUsage()
  n63 = Add(A=Multiply(A=Normalize(VectorInput=@n59), B=LinearInterpolate(A=@n60, B=@n61, Alpha=@n62)), B=NamedRerouteUsage())
  n64 = NamedRerouteUsage()
  n65 = Transform(None=Add(A=NamedRerouteDeclaration(None=@n54.Partcile Spawn Pos) {name=ParticleSpawnPos, nodeColor=<rgba 1 0.8906 0 1>}, B=Multiply(A=@n63, B=@n64)) {desc=Add Velocity in}) {transformSourceType=TRANSFORMSOURCE_Local}
  n66 = FunctionInput(Preview=StaticBool()) {
   bUsePreviewValueAsDefault=true
   desc=Allow for billboard math to be handled outside of this function and to NOT use built in functionality
   inputName=BypassBaseBillboardFunc
   inputType=FunctionInput_StaticBool
   sortPriority=6
  }
  NamedRerouteDeclaration(None=StaticSwitch(True=@n65, False=Add(A=@n52, B=@n65), Value=@n66)) {name=WPO_OUT, nodeColor=<rgba 0.9219 1 0 1>}
  NamedRerouteDeclaration(None=FunctionInput(Preview=StaticBool()) {bUsePreviewValueAsDefault=true, inputName=UseCurveForColorOpacity, inputType=FunctionInput_StaticBool, sortPriority=7}) {name=UseCustomCurveBool, nodeColor=<rgba 0 0.25 1 1>}
  n67 = MaterialFunctionCall(f3=NamedRerouteUsage()) {materialFunction=&AddComponents}
  n68 = ScalarParameter() {group=Color, parameterName=ParticleHueShiftRange, sliderMax=1}
  n69 = NamedRerouteUsage()
  n70 = LinearInterpolate(
   A=VectorParameter() {group=Color, parameterName=Color Start, defaultValue=<rgba 0 0 0 1>}.RGB
   B=VectorParameter() {group=Color, parameterName=Color End, defaultValue=<rgba 0 0 0 1>}.RGB
   Alpha=NamedRerouteUsage()
  )
  n71 = NamedRerouteUsage()
  NamedRerouteDeclaration(None=MaterialFunctionCall(Hue Shift Percentage=Multiply(A=@n67.f3, B=@n68), Texture=StaticSwitch(True=@n69, False=@n70, Value=@n71)) {materialFunction=&HueShift}.Result) {name=Color, nodeColor=<rgba 1 0 0.6562 1>}
  NamedRerouteDeclaration(None=ComponentMask(None=@n25) {b=true, g=true, r=true}) {name=CurveColor, nodeColor=<rgba 0 1 0.2266 1>}
  NamedRerouteUsage()
 ]
}
