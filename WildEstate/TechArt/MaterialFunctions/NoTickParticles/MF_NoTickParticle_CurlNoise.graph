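# Material function that adds curl noise motion to the World Position Offset of NoTick particles.
# It samples a tiling curl noise texture with tangent-to-local transformed UVs, panned by time plus a per-particle offset from random color to reduce clumping.
# The sample is centered on zero, scaled by force and converted from local to world space.
# Static switches optionally multiply it by lifetime, or bypass the effect entirely.
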MaterialFunction /WildEstate/TechArt/MaterialFunctions/NoTickParticles/MF_NoTickParticle_CurlNoise
graph {nodes [
  n1 = FunctionInput() {inputName=WPO}
  n2 = Transform(None=AppendVector(A=TextureCoordinate(), B=Constant())) {transformType=TRANSFORM_Local}
  n3 = FunctionInput() {bUsePreviewValueAsDefault=true, inputName=ParticleRandomColor, sortPriority=3}
  n4 = ScalarParameter() {
   defaultValue=10
   desc=Per Particle Time offset. larger values may reduce clumping for slower particles
   group=Curl Noise
   parameterName=TimeSpread
   sortPriority=15
  }
  n5 = Add(
   A=FunctionInput(Preview=Time()) {bUsePreviewValueAsDefault=true, inputName=Time, inputType=FunctionInput_Scalar, sortPriority=5}
   B=Multiply(A=MaterialFunctionCall(f3=@n3) {materialFunction=&AddComponents}.f3, B=@n4) {constB=11.567}
  )
  n6 = ScalarParameter() {group=Curl Noise, parameterName=CurlNoiseSpeed, sortPriority=13}
  n7 = TextureSample(UVs=Add(A=ComponentMask(None=@n2) {g=true, r=true}, B=Multiply(A=@n5, B=@n6))) {samplerType=SAMPLERTYPE_LinearColor, texture=/Volumetrics/Content/VolumeTextures/Textures/T_TilingSimpleCurlNoise}
  n8 = Transform(None=Multiply(A=Subtract(A=@n7.RGB) {constB=0.5}, B=ScalarParameter() {group=Curl Noise, parameterName=CurlNoiseForce, sortPriority=12}) {constB=2}) {transformSourceType=TRANSFORMSOURCE_Local}
  n9 = StaticSwitchParameter(True=Multiply(A=@n8, B=FunctionInput() {inputName=Lifetime, inputType=FunctionInput_Scalar, sortPriority=1}), False=@n8) {group=Curl Noise, parameterName=CurlNoiseIsLifetimeDependant, sortPriority=40}
  FunctionOutput(None=StaticSwitchParameter(True=Add(A=@n1, B=@n9), False=@n1) {group=Curl Noise, parameterName=UseCurlNoise, sortPriority=20}) {outputName=WPO}
 ]}
