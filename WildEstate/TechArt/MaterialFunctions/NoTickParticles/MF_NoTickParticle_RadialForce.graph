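# Material function for no-tick particles that adds a radial force to World Position Offset, with positive values pulling toward a center.
# It scales the direction from channel-masked particle position to center by force and lifetime, transforms it from local space, and adds it to incoming WPO.
# DistanceAlpha is saturated travel over center distance, reaching one at center although its description says zero.
# UseRadialForce statically toggles everything and drives OnOffFloat.
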MaterialFunction /WildEstate/TechArt/MaterialFunctions/NoTickParticles/MF_NoTickParticle_RadialForce
graph {nodes [
  n1 = FunctionInput() {inputName=WPO In}
  n2 = VectorParameter() {group=Radial Force, parameterName=Radial Force Channel Mask, sortPriority=33, defaultValue=<rgba 1 1 1 1>}
  n3 = Subtract(
   A=FunctionInput() {bUsePreviewValueAsDefault=true, inputName=Radial Force Center, sortPriority=6}
   B=Multiply(A=FunctionInput() {inputName=ParticlePosition, sortPriority=5}, B=@n2.RGB)
  )
  n4 = Multiply(
   A=Multiply(A=Normalize(VectorInput=@n3), B=ScalarParameter() {group=Radial Force, parameterName=Radial Force})
   B=FunctionInput() {inputName=Lifetime, inputType=FunctionInput_Scalar, sortPriority=1}
  )
  n5 = Transform(None=@n4) {transformSourceType=TRANSFORMSOURCE_Local}
  FunctionOutput(None=StaticSwitchParameter(True=Add(A=@n1, B=@n5), False=@n1) {group=Radial Force, parameterName=UseRadialForce, sortPriority=20}) {outputName=WPO}
  n6 = Divide(A=Length(None=@n4) {is=class MaterialExpressionLength}, B=Length(None=@n3)) {constB=1}
  FunctionOutput(None=StaticSwitchParameter(True=Saturate(None=@n6), False=Constant() {r=1}) {group=Radial Force, parameterName=UseRadialForce, sortPriority=20}) {
   description="Alpha from 0 -1. Zero being directly at center or PASSED center point.\r\n\r\nUse this to cull particles traveling inwards toward the center to hide them/WPO scale them down. Otherwise they will continue traveling passed the center"
   outputName=DistanceAlpha
   sortPriority=2
  }
  FunctionOutput(None=StaticSwitchParameter(True=Constant() {r=1}, False=Constant()) {group=Radial Force, parameterName=UseRadialForce, sortPriority=20}) {outputName=OnOffFloat, sortPriority=4}
 ]}
