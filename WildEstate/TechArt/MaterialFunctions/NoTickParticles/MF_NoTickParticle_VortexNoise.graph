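# Material function for NoTick particles that optionally adds a vortex rotation offset to incoming world position offset.
# RotateAboutWorldAxis_cheap spins the world-transformed spawn position around the object position plus a pivot offset, with a parameter selecting X, Y or Z axis.
# Per-particle random color varies phase and speed over time.
# Static switches enable the effect and optionally scale it by Lifetime; no noise is sampled despite the name.
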MaterialFunction /WildEstate/TechArt/MaterialFunctions/NoTickParticles/MF_NoTickParticle_VortexNoise
graph {nodes [
  n1 = NamedRerouteUsage()
  n2 = ScalarParameter() {
   desc="Axis Selection (0 = X, 1 = Y, 2 = Z) in world space\r\n"
   group=Vortex Force
   parameterName=Vortex Axis Switch
   sliderMax=2
   sortPriority=20
  }
  n3 = FunctionInput() {inputName=ParticleRandColor, sortPriority=2}
  n4 = Frac(None=@n3)
  n5 = ScalarParameter() {group=Vortex Force, parameterName=VortexRotationOffset}
  n6 = ScalarParameter() {defaultValue=1, desc=Multiplied by Time, group=Vortex Force, parameterName=VortexPerParticleSpeedVarianceMin, sortPriority=34}
  n7 = ScalarParameter() {defaultValue=1, desc=Multiplied by Time, group=Vortex Force, parameterName=VortexPerParticleSpeedVarianceMax, sortPriority=35}
  n8 = LinearInterpolate(
   A=@n6
   B=@n7
   Alpha=Frac(None=MaterialFunctionCall(f3=@n3) {materialFunction=&AddComponents}.f3)
  )
  n9 = Multiply(
   A=FunctionInput(Preview=Time()) {bUsePreviewValueAsDefault=true, inputName=Time, inputType=FunctionInput_Scalar, sortPriority=5}
   B=ScalarParameter() {group=Vortex Force, parameterName=VortexRotationSpeed, sortPriority=33}
  )
  n10 = VectorParameter() {desc=World Space Offset, group=Vortex Force, parameterName=VortexPivotOffset, sortPriority=39, defaultValue=<rgba 0 0 0 1>}
  MaterialExpressionMaterialFunctionCall_2 = MaterialFunctionCall!(
   Rotation Amount=Add(A=Add(A=@n4, B=@n5), B=Multiply(A=@n8, B=@n9))
   PivotPoint=Add(A=ObjectPositionWS(), B=@n10.RGB)
   WorldPosition=TransformPosition(None=FunctionInput() {inputName=ParticleSpawnPos, sortPriority=4}) {transformType=TRANSFORMPOSSOURCE_World}
  ) {materialFunction=/Engine/Functions/Engine_MaterialFunctions02/WorldPositionOffset/RotateAboutWorldAxis_cheap}
  n11 = Switch!(
   SwitchValue=@n2
   Default
   X=@MaterialExpressionMaterialFunctionCall_2.X-Axis
   Y=@MaterialExpressionMaterialFunctionCall_2.X-Axis
   Z=@MaterialExpressionMaterialFunctionCall_2.X-Axis
  ) {
   description=AxisSelection
   inputs [
    {
     inputName=X
     input {
      expression=&self:MaterialExpressionMaterialFunctionCall_2
      inputName=None
      mask=0
      maskA=0
      maskB=0
      maskG=0
      maskR=0
      outputIndex=0
     }
    }
    {
     inputName=Y
     input {
      expression=&self:MaterialExpressionMaterialFunctionCall_2
      inputName=None
      mask=0
      maskA=0
      maskB=0
      maskG=0
      maskR=0
      outputIndex=1
     }
    }
    {
     inputName=Z
     input {
      expression=&self:MaterialExpressionMaterialFunctionCall_2
      inputName=None
      mask=0
      maskA=0
      maskB=0
      maskG=0
      maskR=0
      outputIndex=2
     }
    }
   ]
  }
  n12 = StaticSwitchParameter(True=Multiply(A=FunctionInput() {inputName=Lifetime, inputType=FunctionInput_Scalar, sortPriority=1}, B=@n11), False=@n11) {group=Vortex Force, parameterName=VortexForceIsLifetimeDependant, sortPriority=40}
  FunctionOutput(None=StaticSwitchParameter(True=Add(A=@n1, B=@n12), False=@n1) {group=Vortex Force, parameterName=UseVortexForce, sortPriority=1}) {outputName=WPO}
  NamedRerouteDeclaration(None=FunctionInput() {inputName=WPO}) {name=WPO In, nodeColor=<rgba 1 0.5625 0 1>}
 ]}
