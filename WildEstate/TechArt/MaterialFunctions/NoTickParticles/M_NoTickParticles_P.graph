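# Translucent unlit material for "NoTick" particles, which builds particle motion as World Position Offset from shared material functions.
# MF_NoTickParticleBase supplies color, opacity, spawn position, time, lifetime, per-particle random and base WPO through named reroutes.
# A modifier chain runs WPO through radial force, vortex noise and curl noise in series, publishing the result as ProcessedWPO.
# Radial force also exposes its on/off and distance-alpha values as named reroutes.
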Material /WildEstate/TechArt/MaterialFunctions/NoTickParticles/M_NoTickParticles_P
graph {
 properties {bAutomaticallySetUsageInEditor=false, blendMode=BLEND_Translucent, shadingModel=MSM_Unlit, usedShadingModels=Unlit}
 nodes [
  n1 = MaterialFunctionCall!(WPO In=NamedRerouteUsage(), Lifetime=NamedRerouteUsage(), ParticlePosition=NamedRerouteUsage(), Radial Force Center) {materialFunction=&/MF_NoTickParticle_RadialForce}
  NamedRerouteDeclaration(None=@n1.OnOffFloat) {name=RadialForce_OnOff, nodeColor=<rgba 1 0.972 0.0704 1>}
  NamedRerouteDeclaration(None=@n1.DistanceAlpha) {name=RadialForce_DistanceAlpha, nodeColor=<rgba 1 0.972 0.0704 1>}
  n2 = MaterialFunctionCall!(UVs, BypassBaseBillboardFunc, UseBillboardWPO, UseCurveForColorOpacity, BoundModifier, WindModifierMult, UseDepthFadeInOpacity) {materialFunction=&/MF_NoTickParticleBase}
  NamedRerouteDeclaration(None=@n2.Color) {name=Base_Color, nodeColor=<rgba 1 0 0.0573 1>}
  NamedRerouteDeclaration(None=@n2.ParticleSpawnPosition) {name=Base_ParticleSpawnPos, nodeColor=<rgba 1 0 0.0573 1>}
  NamedRerouteDeclaration(None=@n2.Time) {name=Base_Time, nodeColor=<rgba 1 0 0.0573 1>}
  NamedRerouteDeclaration(None=@n2.WPO) {name=Base_WPO, nodeColor=<rgba 1 0 0.0573 1>}
  NamedRerouteDeclaration(None=@n2.Lifetime) {name=Base_ParticleLifetime, nodeColor=<rgba 1 0 0.0573 1>}
  n3 = NamedRerouteUsage()
  n4 = NamedRerouteUsage()
  n5 = NamedRerouteUsage()
  n6 = NamedRerouteUsage()
  NamedRerouteDeclaration(None=MaterialFunctionCall!(
    WPO=MaterialFunctionCall!(WPO=@n1.WPO, Lifetime=@n3, ParticleRandColor=@n4, ParticleSpawnPos=@n5, Time=@n6) {materialFunction=&/MF_NoTickParticle_VortexNoise}.WPO
    Lifetime=NamedRerouteUsage()
    ParticleRandomColor=NamedRerouteUsage()
    Time=NamedRerouteUsage()
   ) {materialFunction=&/MF_NoTickParticle_CurlNoise}.WPO) {name=ProcessedWPO, nodeColor=<rgba 1 0 0.8906 1>}
  NamedRerouteDeclaration(None=@n2.Opacity) {name=Base_ParticleOpacity, nodeColor=<rgba 1 0 0.0573 1>}
  NamedRerouteDeclaration(None=@n2.PerParticleRandom) {name=Base_ParticleRandomColor, nodeColor=<rgba 1 0 0.0573 1>}
  NamedRerouteUsage()
  NamedRerouteUsage()
  NamedRerouteUsage()
 ]
}
