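# Material function for no-tick particles that computes a scale multiplier from where each particle sits around the object pivot relative to the camera.
# The local spawn position becomes a direction, is dotted with the camera-to-pivot vector, then remapped and smoothstepped into a tunable mask.
# The mask blends min and max multipliers, enlarging far-side particles by default, and is also output raw.
# A camera-space Power/Lerp branch is left unconnected.
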MaterialFunction /WildEstate/TechArt/MaterialFunctions/NoTickParticles/MF_NoTickParticles_BoundingBoxScaling
graph {nodes [
  CameraVectorWS()
  n1 = Normalize(VectorInput=Transform(None=FunctionInput() {inputName=PartcileSpawnPos, previewValue=<100 50 0 1>}) {transformSourceType=TRANSFORMSOURCE_Local}) {desc=Direction particle is from Object Center}
  n2 = Normalize(VectorInput=Subtract(
    A=MaterialFunctionCall() {materialFunction=&ObjectPivotPoint}.Object Pivot Location
    B=CameraPositionWS()
   ))
  n3 = SmoothStep(
   Min=ScalarParameter() {group=Scale, parameterName=BBXScale.MaskPowerMin}
   Max=ScalarParameter() {defaultValue=1, group=Scale, parameterName=BBXScale.MaskPowerMax}
   Value=ConstantBiasScale(None=DotProduct(A=@n1, B=@n2))
  )
  FunctionOutput(None=LinearInterpolate(
    A=ScalarParameter() {defaultValue=0.7, group=Scale, parameterName=BBXScale.MinScaleMult}
    B=ScalarParameter() {defaultValue=1.2, group=Scale, parameterName=BBXScale.MaxScaleMult}
    Alpha=@n3
   )) {description=Scale Values, outputName=Scale Multiplier}
  FunctionOutput(None=@n3) {outputName=Mask without scale values, sortPriority=1}
  n4 = Constant3Vector() {constant=<rgba 1 0 0 0>}
  n5 = DotProduct(A=Transform(None=Normalize(VectorInput=@n4)) {transformSourceType=TRANSFORMSOURCE_Camera}, B=Constant3Vector() {constant=<rgba 0 0 1 0>})
  LinearInterpolate(Alpha=Power(Base=ConstantBiasScale(None=@n5)) {constExponent=5})
 ]}
