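# Material function for no-tick particles that outputs a shading mask based on each particle's position relative to the object center and the sun.
# The spawn position input is transformed from local space and normalized into a direction from the center, then dotted with the atmospheric light vector.
# The biased and scaled result blends between the BBXShading MinValue and MaxValue parameters, faking directional lighting across the particle volume.
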MaterialFunction /WildEstate/TechArt/MaterialFunctions/NoTickParticles/MF_NoTickParticles_BoundingBoxShading
graph {nodes [
  n1 = FunctionInput() {inputName=PartcileSpawnPos}
  n2 = DotProduct(
   A=Normalize(VectorInput=Transform(None=@n1) {transformSourceType=TRANSFORMSOURCE_Local}) {desc=Direction particle is from Object Center}
   B=AtmosphericLightVector()
  )
  FunctionOutput(None=LinearInterpolate(
    A=ScalarParameter() {group=Shading, parameterName=BBXShading.MinValue, sortPriority=20}
    B=ScalarParameter() {defaultValue=1, group=Shading, parameterName=BBXShading.MaxValue, sortPriority=20}
    Alpha=ConstantBiasScale(None=@n2)
   )) {outputName=Mask}
 ]}
